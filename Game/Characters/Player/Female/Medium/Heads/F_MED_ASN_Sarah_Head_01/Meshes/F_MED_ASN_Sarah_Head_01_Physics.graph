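# Physics asset for the Sarah 01 head mesh, a medium female player character head, set up for physics on three dynamic hair bones.
# A large capsule covers the head, and a small capsule sits on each hair bone.
# Each hair body is constrained directly to the head rather than chained, with narrow swing limits.
# Twist is nearly locked on the first and third hair bones and looser on the second.
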PhysicsAsset /Game/Characters/Player/Female/Medium/Heads/F_MED_ASN_Sarah_Head_01/Meshes/F_MED_ASN_Sarah_Head_01_Physics
bodies {
 dyn_hair_01 [{shapeType=Capsule, length=4.0533, radius=0.5, center=<1.6415 0 0>, rotation=<rot 90 0 0>}]
 dyn_hair_02 [{shapeType=Capsule, length=2.8789, radius=0.505, center=<1.0069 -0.7186 0.0002>, rotation=<rot 90 0 38.3291>}]
 dyn_hair_03 [{shapeType=Capsule, length=6.4959, radius=0.9031, center=<3.3267 0 0>, rotation=<rot 90 0 0>}]
 head [{shapeType=Capsule, length=7.25, radius=6.75, center=<2 2 0>, rotation=<rot 90 0 0>}]
}
constraints [
 {bone1Name=dyn_hair_01, bone2Name=head, swing1LimitDegrees=15, swing2LimitDegrees=15, twistLimitDegrees=2}
 {bone1Name=dyn_hair_02, bone2Name=head, swing1LimitDegrees=15, swing2LimitDegrees=15, twistLimitDegrees=15}
 {bone1Name=dyn_hair_03, bone2Name=head, swing1LimitDegrees=15, swing2LimitDegrees=15, twistLimitDegrees=2}
]
