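# Post-style Anim Blueprint for the "zom" variant of the Sarah 01 female head, adding secondary physics motion on the shared head skeleton.
# The cached input pose runs through two chained RigidBody nodes: one using Sarah's physics asset, then a quarter-weight pass borrowing Hector's zom asset.
# An AnimDynamics hair chain, a jaw offset branch and a bool blend are present but never reach the output.
# Initialization stores the owning mesh in MeshToCopy, which nothing reads.
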AnimBlueprint /BRCosmetics/Characters/Player/Female/Medium/Heads/F_MED_ASN_Sarah_Head_01/Meshes/F_MED_ASN_Sarah_Head_01_Dyn_AnimBP_zom
target_skeleton=/Game/Characters/Player/Common/Fortnite_Base_Head/Fortnite_Base_Head_Export_Skeleton
graphs [
 AnimGraph {nodes [
   n1 = Anim Dynamics - Bone: dyn_hair_01(Alpha=0.33) {
    bGravityOverrideInSimSpace=true
    alphaBoolBlend {blendOption=Linear}
    boundBone {boneName=dyn_hair_01}
    physicsBodyDefinitions [{
      sphereCollisionRadius=0
      boundBone {boneName=dyn_hair_01}
      boxExtents=<20 20 20>
      constraintSetup {angularLimitsMax=<0 61.1999 61.1999>, angularLimitsMin=<0 -57.5999 -43.2>, angularTarget=<0 0 0>}
      localJointOffset=<3 0 0>
     }]
   }
   n2 = Anim Dynamics - Bone: dyn_hair_02(ComponentPose=@n1, Alpha=0.33) {
    bGravityOverrideInSimSpace=true
    alphaBoolBlend {blendOption=Linear}
    boundBone {boneName=dyn_hair_02}
    physicsBodyDefinitions [{
      sphereCollisionRadius=0
      boundBone {boneName=dyn_hair_02}
      boxExtents=<20 20 20>
      constraintSetup {angularLimitsMax=<0 61.1999 61.1999>, angularLimitsMin=<0 -57.5999 -43.2>, angularTarget=<0 0 0>}
      localJointOffset=<3 0 0>
     }]
   }
   n3 = Anim Dynamics - Bone: dyn_hair_03(ComponentPose=@n2, Alpha=0.33, ExternalForce="0,0,0") {
    bGravityOverrideInSimSpace=true
    alphaBoolBlend {blendOption=Linear}
    boundBone {boneName=dyn_hair_03}
    physicsBodyDefinitions [{
      sphereCollisionRadius=0
      boundBone {boneName=dyn_hair_03}
      boxExtents=<20 20 20>
      constraintSetup {angularLimitsMax=<0 61.1999 61.1999>, angularLimitsMin=<0 -57.5999 -43.2>, angularTarget=<0 0 0>}
      localJointOffset=<3 0 0>
     }]
   }
   n4 = Anim Dynamics - Bone: dyn_hair_04(ComponentPose=@n3, Alpha=.33, ExternalForce="0,0,0") {
    bGravityOverrideInSimSpace=true
    alphaBoolBlend {blendOption=Linear}
    boundBone {boneName=dyn_hair_04}
    physicsBodyDefinitions [{
      sphereCollisionRadius=0
      boundBone {boneName=dyn_hair_04}
      boxExtents=<20 20 20>
      constraintSetup {angularLimitsMax=<0 0 0>, angularLimitsMin=<0 0 0>, angularTarget=<0 0 0>, linearAxesMax=<0.5 0.5 0.5>, linearAxesMin=<-0.5 -0.5 -0.5>}
      localJointOffset=<3 0 0>
     }]
   }
   n5 = Anim Dynamics - Bone: dyn_hair_05(ComponentPose=@n4, Alpha=0.33, ExternalForce="0,0,0") {
    bGravityOverrideInSimSpace=true
    alphaBoolBlend {blendOption=Linear}
    boundBone {boneName=dyn_hair_05}
    physicsBodyDefinitions [{
      sphereCollisionRadius=0
      boundBone {boneName=dyn_hair_05}
      boxExtents=<20 20 20>
      constraintSetup {angularLimitsMax=<0 0 0>, angularLimitsMin=<0 0 0>, angularTarget=<0 0 0>, linearAxesMax=<0.1 0.1 0.1>, linearAxesMin=<-0.1 -0.1 -0.1>}
      localJointOffset=<0 0 0>
     }]
   }
   n6 = Anim Dynamics - Bone: dyn_hair_05_a(ComponentPose=@n5, Alpha=1, ExternalForce="0,0,0") {
    bGravityOverrideInSimSpace=true
    alphaBoolBlend {blendOption=Linear}
    boundBone {boneName=dyn_hair_05_a}
    physicsBodyDefinitions [{
      sphereCollisionRadius=0
      boundBone {boneName=dyn_hair_05_a}
      boxExtents=<20 20 20>
      constraintSetup {angularLimitsMax=<0 0 25.2>, angularLimitsMin=<0 -64.7999 -21.6>, angularTarget=<0 0 0>}
      localJointOffset=<3 0 0>
     }]
   }
   n7 = Anim Dynamics - Bone: dyn_hair_05_b(ComponentPose=@n6, Alpha=0.6, ExternalForce="0,0,0") {
    bGravityOverrideInSimSpace=true
    alphaBoolBlend {blendOption=Linear}
    boundBone {boneName=dyn_hair_05_b}
    physicsBodyDefinitions [{
      sphereCollisionRadius=0
      boundBone {boneName=dyn_hair_05_b}
      boxExtents=<20 20 20>
      constraintSetup {angularLimitsMax=<0 2 16.4>, angularLimitsMin=<0 -70.4 -16.4>, angularTarget=<0 0 0>}
      localJointOffset=<3 0 0>
     }]
   }
   Anim Dynamics - Bone: dyn_hair_05_c(ComponentPose=@n7, Alpha=0.3, ExternalForce="0,0,0") {
    bGravityOverrideInSimSpace=true
    alphaBoolBlend {blendOption=Linear}
    boundBone {boneName=dyn_hair_05_c}
    physicsBodyDefinitions [{
      sphereCollisionRadius=0
      boundBone {boneName=dyn_hair_05_c}
      boxExtents=<20 20 20>
      constraintSetup {angularLimitsMax=<0 61.1999 61.1999>, angularLimitsMin=<0 -57.5999 -43.2>, angularTarget=<0 0 0>}
      localJointOffset=<3 0 0>
     }]
   }
   Blend Poses by bool(BlendPose_0=Use cached pose 'Sub Graph Cache'(), bActiveValue=Get bCanPlayCustomAnimations() -> (bCanPlayCustomAnimations:bool)) {blendType=Linear}
   n8 = RigidBody() {
    bClampLinearTranslationLimitToRefPose=true
    bOverrideWorldGravity=true
    simulationSpace=BaseBoneSpace
    alphaBoolBlend {blendOption=Linear}
    baseBoneRef {boneName=spine_05}
    componentLinearAccScale=<1 1 1>
    componentLinearVelScale=<1 1 1>
    overrideWorldGravity {z=-980}
   }
   "Transform (Modify) Bone - Bone: C_jaw"(ComponentPose=@n8, Alpha=1, Translation="0,-1.2,0", Rotation="0,0,0", Scale="1,1,1") {translationMode=BMM_Additive, alphaBoolBlend {blendOption=Linear}, boneToModify {boneName=C_jaw}, translation {y=-1.2}}
   n9 = RigidBody(ComponentPose=Local To Component(LocalPose=Use cached pose 'Sub Graph Cache'())) {
    bClampLinearTranslationLimitToRefPose=true
    bOverrideWorldGravity=true
    overridePhysicsAsset=/Game/Characters/Player/Female/Medium/Heads/F_MED_ASN_Sarah_Head_01/Meshes/F_MED_ASN_Sarah_Head_01_Physics
    simulationSpace=BaseBoneSpace
    alphaBoolBlend {blendOption=Linear}
    baseBoneRef {boneName=root}
    componentAppliedLinearAccClamp=<2000 2000 2000>
    componentLinearAccScale=<0.3 0.3 0.3>
    componentLinearVelScale=<0.2 0.2 0.2>
    overrideWorldGravity {z=-980}
   }
   n10 = RigidBody(ComponentPose=@n9, Alpha=0.25) {
    bClampLinearTranslationLimitToRefPose=true
    bOverrideWorldGravity=true
    overridePhysicsAsset=/Game/Characters/Player/Male/Medium/Heads/M_MED_HIS_Hector_Head_01/Meshes/M_MED_Hector_01_Physics_zom
    alphaBoolBlend {blendOption=Linear}
    baseBoneRef {boneName=head}
    componentLinearAccScale=<0.3 0.3 0.3>
    componentLinearVelScale=<0.3 0.3 0.3>
    overrideWorldGravity {z=-980}
   }
   Output Pose(Result=Component To Local(ComponentPose=@n10)) {layerGroup=None}
   AnimGraphNode_SubInput_0 = Input Pose()
   Save cached pose 'Sub Graph Cache'(Pose=@AnimGraphNode_SubInput_0) {cacheName=Sub Graph Cache}
  ]}
 EventGraph {nodes [
   Event BlueprintInitializeAnimation()
   > Set MeshToCopy(MeshToCopy:Skeletal Mesh Component Object Reference=GetOwnerSkeletalMeshComponent()) -> (Output_Get:Skeletal Mesh Component Object Reference)
   TryGetPawnOwner()
   Event BlueprintUpdateAnimation()
  ]}
]
variables [MeshToCopy {type=SkeletalMeshComponent}]
